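# Masked material for the Wreath cosmetic vehicle wheel, layered on the shared simple wheel parent function and usable on skeletal meshes.
# Opacity mask is replaced by vertex red plus the rim mask texture, so red-painted areas stay opaque while the texture cuts out the rest.
# Vertex red also blends the parent's emissive toward a tire emissive texture scaled by a glow parameter.
# The result passes through the shared vehicle damage function, then the de-rez transition.
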Material /VehicleCosmetics/Wheels/Wreath/Materials/M_Wheel_Wreath
graph {
 properties {
  bAutomaticallySetUsageInEditor=false
  bUseMaterialAttributes=true
  bUsedWithMeshDeformer=true
  bUsedWithSkeletalMesh=true
  blendMode=BLEND_Masked
  pixelDepthOffsetMode=PDOM_Legacy
  usedShadingModels=Default Lit
 }
 nodes [
  Constant()
  n1 = GetMaterialAttributes(None=MaterialFunctionCall() {materialFunction=&MF_WheelParent_Simple}.MaterialAttributes) {attributeGetTypes [#1, #2]}
  n2 = VertexColor()
  n3 = TextureSampleParameter2D() {
   group=5_Special Effects
   parameterName=Opacity_Mask
   samplerType=SAMPLERTYPE_LinearColor
   texture=/VehicleCosmetics/Wheels/Wreath/Textures/T_Wreath_Rim_M
  }
  n4 = LinearInterpolate(A=@n1.Opacity Mask, B=Add(A=@n2.R, B=@n3.R), Alpha=Constant() {r=1})
  n5 = TextureSampleParameter2D() {
   group=5_Special Effects
   parameterName=Additional_Emissive
   samplerType=SAMPLERTYPE_LinearColor
   texture=/VehicleCosmetics/Wheels/Wreath/Textures/T_Wreath_Tire_E
  }
  n6 = LinearInterpolate(
   A=@n1.Emissive Color
   B=Multiply(A=@n5.RGB, B=ScalarParameter() {defaultValue=1, group=5_Special Effects, parameterName=AddEmissive_Glow})
   Alpha=@n2.R
  )
  MaterialFunctionCall(In=MaterialFunctionCall(Material=SetMaterialAttributes!(MaterialAttributes=@n1.MaterialAttributes, Opacity Mask=@n4, Emissive Color=@n6) {attributeSetTypes [#1, #2]}) {materialFunction=&MF_BaseVehicle_Damage}.Result) {materialFunction=&MF_VehicleDeRez_Transition}
 ]
}
